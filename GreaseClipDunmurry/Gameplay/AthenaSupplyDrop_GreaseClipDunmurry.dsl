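; Supply drop actor for GreaseClipDunmurry that rolls its own loot and throws it out as pickups when opened.
; On BeginPlay it picks drops from its GreaseClipLoot setting and caches them.
; On opened, it spawns a pickup per cached entry at its location and tosses it forward and upward with random horizontal scatter.
; Its interaction check always returns true, and the overlap and tick events are empty.
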(event EventBeginPlay
  (bind _outloottodrop (Loot|PickLootDrops (Variables|Default|GetGreaseClipLoot) 0 -1 "(GameplayTags=)"))
  (Variables|Default|SetOut%20Loot%20to%20Drop _outloottodrop))

(event EventOnOpened
  (bind _self self)
  (for _array_element (Variables|Default|GetOut%20Loot%20to%20Drop)
    (bind _asfort_world_item_definition (Utilities|Casting|CastToFortWorldItemDefinition (Class|FortItemFunctionLibrary|BreakFortItemEntry _array_element)))
    (bind _returnvalue (CallFunction|SpawnPickupfromItemEntry _asfort_world_item_definition _array_element 1 0 (Transformation|GetActorLocation _self) "0.000000,0.000000,0.000000"))
    (Pickup|TossPickupinDirection _returnvalue (Utilities|Operators|Add (* (Transformation|GetActorForwardVector _self) 7.5e+02) (Math|Vector|MakeVector (Math|Random|RandomFloatinRange -1.5e+02 1.5e+02) (Math|Random|RandomFloatinRange -1.5e+02 1.5e+02) 1e+03)) _self false false)))

(event Collision|EventActorBeginOverlap
  (params (OtherActor :type Actor%20Object%20Reference)))

(event EventTick
  (params (DeltaSeconds :type Float%20%28single-precision%29)))

(fn BlueprintCanInteract
  (params (InteractingPawn :type Fort%20Pawn%20Object%20Reference) (InteractionBeingAttempted :type EInteractionBeingAttempted%20Enum :default "FirstInteraction") (InteractionType :type TInteractionType%20Enum :default "IT_NoInteraction"))
  (return true))

(fn ConstructionScript
  (|CallParentFunction))
