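; Lootable chest actor for GreaseClipDunmurry that offers "Retrieve Loot" while CanBeLooted is true, with a search sound during interaction.
; On interact or server death, it waits for playlist data, then on authority selects a loot tier by max squad size and tosses pickups forward and upward with random spread.
; Each path has its own DoOnce, clears CanBeLooted, calls TriggerFX for effects, mesh hiding and audio, then silently dies after a short delay.
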(event Interaction|EventBlueprintOnInterruptInteract
  (Components|Activation|Deactivate (Variables|Default|GetSearchSoundFX)))

(event Interaction|EventBlueprintOnBeginInteract
  (params (InteractingPawn :type Fort%20Pawn%20Object%20Reference) (InteractionBeingAttempted :type EInteractionBeingAttempted%20Enum :default "FirstInteraction") (InteractionType :type TInteractionType%20Enum :default "IT_NoInteraction"))
  (bind _searchsoundfx (Variables|Default|GetSearchSoundFX))
  (if (not (Components|Activation|IsActive _searchsoundfx))
    (Components|Activation|Activate _searchsoundfx false)))

(event Interaction|EventBlueprintOnInteract
  (params (InteractingPawn :type Fort%20Pawn%20Object%20Reference) (InteractionBeingAttempted :type EInteractionBeingAttempted%20Enum :default "FirstInteraction"))
  (bind (_gamestate _playlist) (Class|FortAsyncActionCurrentPlaylistReady|PlaylistDataReadyAsync)
    (:OnReady
      (Networking|SwitchHasAuthority
        (:Authority
          (Utilities|FlowControl|DoOnce false
            (:Completed
              (Variables|Default|SetCanBeLooted false)
              ;[1 1 1] Spawn Loot Drops
              (bind _outloottodrop (Loot|PickLootDrops (Utilities|Select :Index (- (Variables|/Script/FortniteGame.FortPlaylist|GetMaxSquadSize _playlist) 1) :Option%200 "Loot_GCDunmurry_Solo" :Option%201 "Loot_GCDunmurry_Duo" :Option%202 "Loot_GCDunmurry_Trio" :Option%203 "Loot_GCDunmurry_Squad") 0 -1 "(GameplayTags=)"))
              (for _array_element _outloottodrop
                (bind _returnvalue (Inventory|SpawnPickupInWorldWithClassAndItemEntryForInstancedLoot _array_element 0 (Transformation|GetActorLocation self) "0.000000,0.000000,1.000000" 0 false true false "Tossed" "Unset" 0 false))
                (for _array_element_2 _returnvalue
                  (Pickup|TossPickupinDirection _array_element_2 (Utilities|Operators|Add (* (Transformation|GetActorForwardVector self) 7.5e+02) (Math|Vector|MakeVector (Math|Random|RandomFloatinRange -1.5e+02 1.5e+02) (Math|Random|RandomFloatinRange -1.5e+02 1.5e+02) 1e+03)) self false false)))
              ;[1 1 1] Play FX, then despawn
              (CallFunction|TriggerFX)
              (Utilities|FlowControl|Delay 0.2)
              (Gameplay|SilentDie false))))))))

(event EventOnDeathServer
  (params (Damage :type Float%20%28single-precision%29) (DamageTags :type Gameplay%20Tag%20Container%20Structure%20%28by%20ref%29) (Momentum :type Vector :default "0, 0, 0") (HitInfo :type Hit%20Result%20Structure%20%28by%20ref%29) (InstigatedBy :type Controller%20Object%20Reference) (DamageCauser :type Actor%20Object%20Reference) (EffectContext :type Gameplay%20Effect%20Context%20Handle%20Structure))
  (bind (_gamestate _playlist) (Class|FortAsyncActionCurrentPlaylistReady|PlaylistDataReadyAsync)
    (:OnReady
      (Networking|SwitchHasAuthority
        (:Authority
          (Utilities|FlowControl|DoOnce false
            (:Completed
              (Variables|Default|SetCanBeLooted false)
              ;[1 1 1] Spawn Loot Drops
              (bind _outloottodrop (Loot|PickLootDrops (Utilities|Select :Index (- (Variables|/Script/FortniteGame.FortPlaylist|GetMaxSquadSize _playlist) 1) :Option%200 "Loot_GCDunmurry_Solo" :Option%201 "Loot_GCDunmurry_Duo" :Option%202 "Loot_GCDunmurry_Trio" :Option%203 "Loot_GCDunmurry_Squad") 0 -1 "(GameplayTags=)"))
              (for _array_element _outloottodrop
                (bind _returnvalue (Inventory|SpawnPickupInWorldWithClassAndItemEntryForInstancedLoot _array_element 0 (Transformation|GetActorLocation self) "0.000000,0.000000,1.000000" 0 false true false "Tossed" "Unset" 0 false))
                (for _array_element_2 _returnvalue
                  (Pickup|TossPickupinDirection _array_element_2 (Utilities|Operators|Add (* (Transformation|GetActorForwardVector self) 7.5e+02) (Math|Vector|MakeVector (Math|Random|RandomFloatinRange -1.5e+02 1.5e+02) (Math|Random|RandomFloatinRange -1.5e+02 1.5e+02) 1e+03)) self false false)))
              ;[1 1 1] Play FX, then despawn
              (CallFunction|TriggerFX)
              (Utilities|FlowControl|Delay 0.2)
              (Gameplay|SilentDie false))))))))

(event Custom|TriggerFX
  ;[1 1 1] Play FX, then despawn
  (Components|Activation|Activate (Variables|Default|GetLootedFX) false)
  (Rendering|SetVisibility (Variables|Default|GetStaticMesh) false false)
  (Audio|Components|Audio|FadeOut (Variables|Default|GetGreaseClipDunmurry_LootChest_Loop) 0.5 0.0 "Linear" :!class /Script/Engine.AudioComponent)
  (bind _open_chest_sound (Variables|Default|GetOpen%20Chest%20Sound)
    (:then
      (Audio|PlaySoundatLocation _open_chest_sound (Transformation|GetActorLocation)))
    (:else)))

(fn BlueprintCanInteract
  (params (InteractingPawn :type Fort%20Pawn%20Object%20Reference) (InteractionBeingAttempted :type EInteractionBeingAttempted%20Enum :default "FirstInteraction") (InteractionType :type TInteractionType%20Enum :default "IT_NoInteraction"))
  (return (Variables|Default|GetCanBeLooted)))

(fn BlueprintGetInteractionString
  (params (InteractingPawn :type Fort%20Pawn%20Object%20Reference) (InteractionBeingAttempted :type EInteractionBeingAttempted%20Enum :default "FirstInteraction"))
  (return "NSLOCTEXT(\"[8AB03AC11B7403CC677503AADA7E92C0]\", \"85400A47448274CE5D5FC2847AB19FD3\", \"Retrieve Loot\")"))

(fn ConstructionScript)
