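; Lootable workbench Blueprint that gives out one GreaseClipDunmurry weapon item pickup when looted.
; On EventOnLoot it spawns the pickup at the actor's location and tosses it along a suggested projectile arc aimed at the Arrow component.
; It then hides Prop1 and deactivates its AthenaSpecialActor component.
; The construction script only calls the parent.
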(event EventOnLoot
  (bind _returnvalue (Inventory|SpawnPickupInWorld "/GreaseClipDunmurry/Gameplay/WID_GreaseClipDunmurry.WID_GreaseClipDunmurry" 1 (Transformation|GetActorLocation self) "0.000000,0.000000,0.000000" 0 true false false -1 "Tossed" "Chest"))
  (bind _outlaunchvelocity (Game|SuggestProjectileVelocityCustomArc (Transformation|GetActorLocation self) (Transformation|GetWorldLocation (Variables|Default|GetArrow))))
  (Pickup|TossPickupinDirection _returnvalue _outlaunchvelocity self false false)
  (Rendering|SetVisibility (Variables|Default|GetProp1) false false)
  (Components|Activation|Deactivate (Variables|Default|GetAthenaSpecialActor)))

(fn ConstructionScript
  (|CallParentFunction))
