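; Thrown-mod projectile for Valet vehicles that, on hitting an eligible vehicle, replaces it with a modded Valet sports car.
; The server handles one hit, checking mod eligibility by tag, while clients just hide the mesh.
; A valid hit destroys the vehicle, spawns the replacement at its transform, broadcasts a cosmetic-swap event, applies the mod and destroys the projectile.
; An invalid hit reverts the projectile to a world pickup.
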(event Collision|EventHit
  (params (MyComp :type Primitive%20Component%20Object%20Reference) (Other :type Actor%20Object%20Reference) (OtherComp :type Primitive%20Component%20Object%20Reference) (bSelfMoved :type Boolean) (HitLocation :type Vector :default "0, 0, 0") (HitNormal :type Vector :default "0, 0, 0") (NormalImpulse :type Vector :default "0, 0, 0") (Hit :type Hit%20Result%20Structure%20%28by%20ref%29))
  (Networking|SwitchHasAuthority
    (:Authority
      ;[1 1 1] Server Effects
      (Utilities|FlowControl|DoOnce false
        (:Completed
          (bind (_isvalid _asfort_athena_vehicle) (CallFunction|IsValidVehicleImpact Hit))
          (if _isvalid
            (Utilities|FlowControl|Sequence
              (:then_1
                (CallFunction|SwapVehicle _asfort_athena_vehicle)))
            (else
              (CallFunction|HandleInvalidImpact Hit))))))
    (:Remote
      ;[1 1 1] Client Effects
      (Rendering|SetVisibility (Variables|Default|GetStaticMesh) false false))))

(fn HandleInvalidImpact
  (params (Hit :type Hit%20Result%20Structure%20%28by%20ref%29))
  (bind (_bblockinghit _binitialoverlap _time _distance _location _impactpoint _normal _impactnormal) (Collision|BreakHitResult Hit))
  (bind _returnvalue (Inventory|SpawnPickupInWorld (Variables|Default|GetRevertToPickupItemDef) 1 _location _impactnormal))
  (Utilities|IsValid _returnvalue
    (:Is%20Valid
      (Actor|DestroyActor))
    (:Is%20Not%20Valid
      (Utilities|Debugging|RaiseScriptError "Failed to revert to pickup item"))))

(fn IsValidVehicleImpact
  (params (Hit :type Hit%20Result%20Structure%20%28by%20ref%29))
  (bind (_bblockinghit _binitialoverlap _time _distance _location _impactpoint _normal _impactnormal _physmat _hitactor) (Collision|BreakHitResult Hit))
  (bind _asfort_athena_vehicle (Utilities|Casting|CastToFortAthenaVehicle _hitactor)
    (:then
      (if (Vehicle|Mod|CanApplyMod _asfort_athena_vehicle (Variables|Default|GetMod_Tag))
        (return true _asfort_athena_vehicle)
        (else
          _)))
    (:CastFailed
      _))
  (return false 0))

(fn SwapVehicle
  (params (VehicleToSwap :type Fort%20Athena%20Vehicle%20Object%20Reference))
  (LocalVariables|SetVehicleTransform (Transformation|GetActorTransform VehicleToSwap))
  (Actor|DestroyActor VehicleToSwap :!pinned true) ; Destroy Old Vehicle
  (Utilities|FlowControl|Sequence
    (:then_0
      (bind _returnvalue (Game|SpawnActorfromClass "/Valet/SportsCar/Valet_SportsCar_Vehicle.Valet_SportsCar_Vehicle_C" (LocalVariables|GetVehicleTransform) "AlwaysSpawn" "MultiplyWithRoot" 0 false 0))
      (Utilities|IsValid _returnvalue
        (:Is%20Valid
          (Utilities|FlowControl|Sequence
            (:then_0
              ;[1 1 1] Vehicle Cosmetics
              (bind _returnvalue_2 (GameplayEventRouter|GetEventRouter "/Valet/BasicCar/Valet_BasicCar_Vehicle.Valet_BasicCar_Vehicle_C" _returnvalue))
              (GameplayEventRouter|BroadcastEvent _returnvalue_2 (Variables|Default|GetVehicleCosmeticSwapPayload) _returnvalue))
            (:then_1
              ;[1 1 1] Vehicle Mod
              (bind _returnvalue_3 (PlayerInfo|GetOwnerPlayerController))
              (Vehicle|Mod|ApplyVehicleMod _returnvalue (Variables|Default|GetMod_Tag) _returnvalue_3))))
        (:Is%20Not%20Valid
          (Utilities|Debugging|RaiseScriptError "Failed to spawn replacement vehicle"))))
    (:then_1
      (Actor|DestroyActor :!pinned true) ; Destroy Projectile (Self)
    )))

(fn ConstructionScript
  (|CallParentFunction))
