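; Vehicle mod actor that makes a boosting car vanish at high speed, then reappear with a loot crate mod and remove itself.
; A server timer checks speed, times the warm-up FX, then runs the travel sequence.
; Hidden-state changes toggle visibility, collision and gameplay effects, pinning velocity at zero by looping timer before restoring the cached values.
; Enter/exit handlers manage passenger cues and squad map-icon tags; entry also swaps engine sound data.
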(event Custom|ToggleVehicleEngineAudio
  (params (VehicleHidden :type Boolean))
  (bind _audiocontroller (Variables|Default|GetAudioController))
  (bind _returnvalue (AudioController|GetEngineAudioComponent _audiocontroller))
  (Networking|CanExecuteCosmeticEvents
    (:True
      (if (Class|FortVehicleAudioDataInterface|IsLocalforAudio (Variables|Default|GetAttachedVehicle))
        _
        (elif VehicleHidden
          (Audio|Components|Audio|FadeOut _returnvalue 0.5 0.0 "Linear" :!class /Script/Engine.AudioComponent)
          (Utilities|Time|SetTimerbyEvent (delegate Custom|FadeIn3PEngineAudio) (ScalableFloat|GetValueatLevel (Variables|Default|GetHF_VehicleHiddenDuration)))
          (elif (Audio|Components|Audio|IsPlaying _returnvalue)
            _
            (else
              (Audio|Components|Audio|FadeIn _returnvalue 0.5 1.0 0.0 "Linear" :!class /Script/Engine.AudioComponent))))))))

(event Custom|FadeIn3PEngineAudio
  (bind _audiocontroller (Variables|Default|GetAudioController))
  (bind _returnvalue (AudioController|GetEngineAudioComponent _audiocontroller))
  (if (not (Audio|Components|Audio|IsPlaying _returnvalue))
    (Audio|Components|Audio|FadeIn _returnvalue 0.5 1.0 0.0 "Linear" :!class /Script/Engine.AudioComponent)))

(event Custom|RestoreVehicleVelocity
  ;[1 1 1] Remove/Restore Velocity
  (bind _asprimitive_component (Utilities|Casting|CastToPrimitiveComponent (Variables|/Script/Engine.Actor|GetRootComponent (Actor|GetOwner))))
  (Physics|SetPhysicsLinearVelocity _asprimitive_component (Variables|Default|GetCachedLinearVelocity) false "None")
  (Physics|SetPhysicsAngularVelocityinDegrees _asprimitive_component (Variables|Default|GetCachedAngularVelocity) false "None"))

(event Custom|PerformTravelSequence
  (bind _hidden_tags (Variables|Default|GetHidden_Tags))
  (bind _attachedvehicle (Variables|Default|GetAttachedVehicle))
  (bind _hidden_tags_2 (Variables|Default|GetHidden_Tags))
  (bind _attachedvehicle_2 (Variables|Default|GetAttachedVehicle))
  ;[1 1 1] Travel Sequence (Server)
  (Utilities|FlowControl|Sequence
    (:then_0
      ;;[1 1 1] Cache Velocity
      (bind _asprimitive_component (Utilities|Casting|CastToPrimitiveComponent (Variables|/Script/Engine.Actor|GetRootComponent (Variables|Default|GetAttachedVehicle)))
        (:then
          (bind _returnvalue (Physics|GetPhysicsLinearVelocity _asprimitive_component "None"))
          (Variables|Default|SetCachedLinearVelocity _returnvalue)
          (Variables|Default|SetCachedAngularVelocity (Physics|GetPhysicsAngularVelocityinDegrees _asprimitive_component "None")))
        (:CastFailed
          (Variables|Default|SetCachedLinearVelocity "0, 0, 0")
          (Variables|Default|SetCachedAngularVelocity "0, 0, 0"))))
    (:then_2
      ;;[0.15 0 0.01365 0.5] To be removed
      ;;[1 1 1] Disappear
      (GameplayCue|ExecuteGameplayCue (Variables|Default|GetAttachedVehicle) (Variables|Default|GetGC_LightningStrikeTemp))
      (GameplayCue|ExecuteGameplayCueOnActor(Burst) (Variables|Default|GetAttachedVehicle) (Variables|Default|GetGC_FireTrails) (Ability|GameplayCue|MakeGameplayCueParameters :NormalizedMagnitude 0.0 :RawMagnitude 0.0 :Location "0,0,0" :Normal "0,0,0" :Instigator self :GameplayEffectLevel 1 :AbilityLevel 1 :bReplicateLocationWhenUsingMinimalRepProxy false))
      (CallFunction|SetVehicleHiddenState true)
      (Ability|GameplayEffect|AddLooseGameplayTags _attachedvehicle _hidden_tags_2 true)
      (for _array_element (Vehicle|GetAllPassengers _attachedvehicle)
        (Ability|GameplayEffect|AddLooseGameplayTags _array_element _hidden_tags_2 true))
      (CallFunction|ToggleVehicleEngineAudio true)
      ;[1 1 1] Reappear
      (Utilities|FlowControl|Delay (ScalableFloat|GetValueatLevel (Variables|Default|GetHF_VehicleHiddenDuration)))
      (GameplayCue|ExecuteGameplayCue (Variables|Default|GetAttachedVehicle) (Variables|Default|GetGC_LightningStrikeTemp))
      (CallFunction|ToggleVehicleEngineAudio false)
      (CallFunction|SetVehicleHiddenState false)
      (Ability|GameplayEffect|RemoveLooseGameplayTags _attachedvehicle_2 _hidden_tags true)
      (for _array_element_2 (Vehicle|GetAllPassengers _attachedvehicle_2)
        (Ability|GameplayEffect|RemoveLooseGameplayTags _array_element_2 _hidden_tags true))
      (Utilities|FlowControl|Delay 0.25)
      (CallFunction|SpawnLoot)
      (Vehicle|Mod|RemoveVehicleMod (Variables|Default|GetAttachedVehicle) (Variables|Default|GetMod%20Tag)))))

(event Custom|HaltVehicle
  ;[1 1 1] Remove/Restore Velocity
  (bind _asprimitive_component (Utilities|Casting|CastToPrimitiveComponent (Variables|/Script/Engine.Actor|GetRootComponent (Actor|GetOwner))))
  (Physics|SetPhysicsLinearVelocity _asprimitive_component "0, 0, 0" false "None")
  (Physics|SetPhysicsAngularVelocityinDegrees _asprimitive_component "0, 0, 0" false "None"))

(event Custom|OnServerCheckSpeed
  (bind _attachedvehicle (Variables|Default|GetAttachedVehicle))
  (bind _returnvalue (Vehicle|GetForwardSpeedKmh _attachedvehicle))
  ;[1 1 1] Server Speed Check
  (Utilities|FlowControl|Sequence
    (:then_0
      (if (and (>= _returnvalue (ScalableFloat|GetValueatLevel (Variables|Default|GetHF_IntroSequenceSpeed))) (Class|FortVehicleInterface|IsBoosting (Variables|Default|GetAttachedVehicle)))
        (if (< (Variables|Default|GetIntroSequenceServerStartTime) 0)
          (Variables|Default|SetIntroSequenceServerStartTime (Fortnite|Sync|Time|GetServerWorldTimeSeconds))
          (CallFunction|ActivateSpeedFX true))
        (else
          (Variables|Default|SetIntroSequenceServerStartTime -1.0)
          (CallFunction|ActivateSpeedFX false))))
    (:then_1
      (if (and (>= _returnvalue (ScalableFloat|GetValueatLevel (Variables|Default|GetHF_TeleportSpeed))) (Class|FortVehicleInterface|IsBoosting (Variables|Default|GetAttachedVehicle)))
        (if (>= (- (Fortnite|Sync|Time|GetServerWorldTimeSeconds) (Variables|Default|GetIntroSequenceServerStartTime)) (ScalableFloat|GetValueatLevel (Variables|Default|GetHF_IntroSequenceDuration)))
          (Utilities|Time|ClearandInvalidateTimerbyHandle (Variables|Default|GetServerSpeedCheckTimer))
          (Variables|Default|SetIntroSequenceServerStartTime -1.0)
          (CallFunction|PerformTravelSequence)
          (CallFunction|ActivateSpeedFX false))))))

(event Custom|ActivateSpeedFX
  (params (Activate :type Boolean))
  (bind _speedfxref (Variables|Default|GetSpeedFXRef))
  (bind _returnvalue (Components|Activation|IsActive _speedfxref))
  ;[1 1 1] HandleSpeedVFX
  (GameplayCue|AddGameplayCueOnActor(Looping) (Variables|Default|GetAttachedVehicle) (Variables|Default|GetGCNL_IntroSequenceLoop) (Ability|GameplayCue|MakeGameplayCueParameters :NormalizedMagnitude 0.0 :RawMagnitude 0.0 :Location "0,0,0" :Normal "0,0,0" :Instigator self :GameplayEffectLevel 1 :AbilityLevel 1 :bReplicateLocationWhenUsingMinimalRepProxy false))
  (Utilities|IsValid _speedfxref
    (:Is%20Valid
      (if Activate
        (if (not _returnvalue)
          (Components|Activation|Activate _speedfxref true)
          (Development|PrintString "ActivateIntro" true true "(R=0.154683,G=1.000000,B=0.000000,A=1.000000)" 5.0))
        (elif _returnvalue
          (Components|Activation|Deactivate _speedfxref)
          (Development|PrintString "DeactivateIntro" true true "(R=1.000000,G=0.494420,B=0.000000,A=1.000000)" 5.0))))))

(event EventBeginPlay
  (bind _attachedvehicle (Variables|Default|GetAttachedVehicle))
  (Utilities|FlowControl|Sequence
    (:then_0
      ;[1 1 1] Cache Owner
      (Variables|Default|SetAttachedVehicle (Utilities|Casting|CastToFortAthenaVehicle (Actor|GetOwner))))
    (:then_1
      ;[1 1 1] Server Speed Check
      (Networking|SwitchHasAuthority
        (:Authority
          (Variables|Default|GetAttachedVehicle
            (:then
              (bind _returnvalue (Utilities|Time|SetTimerbyEvent (EventDispatchers|CreateEvent :!event Custom|OnServerCheckSpeed) (ScalableFloat|GetValueatLevel (Variables|Default|GetHF_ServerSpeedCheckPeriod)) true true))
              (Variables|Default|SetServerSpeedCheckTimer _returnvalue))
            (:else)))))
    (:then_2
      ;[1 1 1] Monitor Passengers
      (bind _attachedvehicle_2 (Variables|Default|GetAttachedVehicle)
        (:then
          (Vehicle|BindEventtoOnPawnEnteredVehicleatSeat _attachedvehicle_2 (EventDispatchers|CreateEvent :!event Custom|HandlePawnEnteredVehicle))
          (Vehicle|BindEventtoOnPawnExitedVehiclefromSeat _attachedvehicle_2 (EventDispatchers|CreateEvent :!event Custom|HandlePawnExitedVehicle))
          (for _array_element (Vehicle|GetAllPassengers _attachedvehicle_2)
            (CallFunction|HandlePawnEnteredVehicle _attachedvehicle_2 _array_element -1)))
        (:else)))
    (:then_3
      (GameplayCue|AddGameplayCueOnActor(Looping) _attachedvehicle (Variables|Default|GetGCNL_ExternalVehicleEffects) (Ability|GameplayCue|MakeGameplayCueParameters :NormalizedMagnitude 0.0 :RawMagnitude 0.0 :Location "0,0,0" :Normal "0,0,0" :Instigator self :GameplayEffectLevel 1 :AbilityLevel 1 :bReplicateLocationWhenUsingMinimalRepProxy false))
      (bind _returnvalue_2 (Niagara|SpawnSystemAttached "/GreaseClipDunmurry/Assets/FX/NS_GreaseClipDunMurry_WarmUp.NS_GreaseClipDunMurry_WarmUp" (Variables|/Script/FortniteGame.FortAthenaVehicle|GetSkeletalMesh _attachedvehicle) "None" "0, 0, 0" "0, 0, 0" "KeepRelativeOffset" false false))
      (bind _output_get (Variables|Default|SetSpeedFXRef _returnvalue_2))
      (Niagara|SetNiagaraVariable(Float) _output_get "IntroDuration" (ScalableFloat|GetValueatLevel (Variables|Default|GetHF_IntroSequenceDuration))))))

(event EventEndPlay
  (params (EndPlayReason :type EEndPlayReason%20Enum :default "Destroyed"))
  ;[1 1 1] Monitor Passengers
  (bind _attachedvehicle (Variables|Default|GetAttachedVehicle)
    (:then
      (Vehicle|UnbindEventfromOnPawnEnteredVehicleatSeat _attachedvehicle (EventDispatchers|CreateEvent :!event Custom|HandlePawnEnteredVehicle))
      (Vehicle|BindEventtoOnPawnExitedVehiclefromSeat _attachedvehicle (EventDispatchers|CreateEvent :!event Custom|HandlePawnExitedVehicle))
      (for _array_element (Vehicle|GetAllPassengers _attachedvehicle)
        (CallFunction|HandlePawnExitedVehicle _attachedvehicle _array_element -1))
      (GameplayCue|RemoveGameplayCueOnActor(Looping) (Variables|Default|GetAttachedVehicle) (Variables|Default|GetGCNL_ExternalVehicleEffects) (Ability|GameplayCue|MakeGameplayCueParameters 0.0)))
    (:else)))

(event Custom|HandlePawnEnteredVehicle
  (params (Vehicle :type Fort%20Vehicle%20Interface%20Interface%20%28by%20ref%29) (PlayerPawn :type Fort%20Pawn%20Object%20Reference) (SeatIndex :type Integer))
  ;[1 1 1] Monitor Passengers
  (bind _asfort_player_pawn (Utilities|Casting|CastToFortPlayerPawn PlayerPawn))
  (Utilities|FlowControl|Sequence
    (:then_0
      (Variables|Default|GetIsVehicleHiddenLocal
        (:then
          (CallFunction|UpdatePassengerHiddenStatus _asfort_player_pawn true))
        (:else)))
    (:then_1
      (if (not (Networking|IsDedicatedServer))
        (Projectile|AddGameplayCue%28Looping%29Local _asfort_player_pawn (Variables|Default|GetGCNL_PassengerEffect) (Ability|GameplayCue|MakeGameplayCueParameters :NormalizedMagnitude 0.0 :RawMagnitude 0.0 :Location "0,0,0" :Normal "0,0,0" :Instigator self :GameplayEffectLevel 1 :AbilityLevel 1 :bReplicateLocationWhenUsingMinimalRepProxy false) false)
        (bind _attachedvehicle (Variables|Default|GetAttachedVehicle)
          (:then
            (AudioControllerConfig|SetVehicleSoundData (Vehicle|GetChildActorbyClass _attachedvehicle "/Script/FortAthenaVehicleAudioRuntime.FortAthenaVehicleAudioController") (Variables|Default|GetVehicleSoundData) true)
            (Variables|Default|SetAudioController (Vehicle|GetChildActorbyClass _attachedvehicle "/Script/FortAthenaVehicleAudioRuntime.FortAthenaVehicleAudioController")))
          (:else))))
    (:then_2
      (bind _asfort_player_controller_athena (Utilities|Casting|CastToFortPlayerControllerAthena (Fortnite|GetFortPlayerControllerfromActor _asfort_player_pawn)))
      (for _array_element (Team|GetSquadMemberPlayerStates _asfort_player_controller_athena true)
        (Ability|GameplayEffect|AddLooseGameplayTags _array_element (Variables|Default|GetMapIconTag) true)))))

(event Custom|HandlePawnExitedVehicle
  (params (Vehicle :type Fort%20Vehicle%20Interface%20Interface%20%28by%20ref%29) (PlayerPawn :type Fort%20Pawn%20Object%20Reference) (SeatIndex :type Integer))
  ;[1 1 1] Monitor Passengers
  (bind _asfort_player_pawn (Utilities|Casting|CastToFortPlayerPawn PlayerPawn))
  (Utilities|FlowControl|Sequence
    (:then_0
      (Variables|Default|GetIsVehicleHiddenLocal
        (:then
          (CallFunction|UpdatePassengerHiddenStatus _asfort_player_pawn))
        (:else
          _))
      (HUD|SetHUDElementVisibilitywithContext (Fortnite|GetFortPlayerControllerfromActor _asfort_player_pawn) (Variables|Default|GetHUDElement%20Tags) false "None" ""))
    (:then_1
      (if (not (Networking|IsDedicatedServer))
        (Projectile|RemoveGameplayCue%28Looping%29Local _asfort_player_pawn (Variables|Default|GetGCNL_PassengerEffect) (Ability|GameplayCue|MakeGameplayCueParameters :NormalizedMagnitude 0.0 :RawMagnitude 0.0 :Location "0,0,0" :Normal "0,0,0" :Instigator self :GameplayEffectLevel 1 :AbilityLevel 1 :bReplicateLocationWhenUsingMinimalRepProxy false) false)
        (Projectile|RemoveGameplayCue%28Looping%29Local _asfort_player_pawn (Variables|Default|GetGCNL_Speedlines) (Ability|GameplayCue|MakeGameplayCueParameters 0.0))))
    (:then_2
      (bind _asfort_player_controller_athena (Utilities|Casting|CastToFortPlayerControllerAthena (Fortnite|GetFortPlayerControllerfromActor _asfort_player_pawn)))
      (for _array_element (Team|GetSquadMemberPlayerStates _asfort_player_controller_athena true)
        (Ability|GameplayEffect|RemoveLooseGameplayTags _array_element (Variables|Default|GetMapIconTag) true)))))

(fn BP_VehicleHiddenChanged
  (CallFunction|UpdateVehicleHiddenStatus))

(fn GetIntroSequencePercentage
  (bind _introsequenceserverstarttime (Variables|Default|GetIntroSequenceServerStartTime))
  (if (< _introsequenceserverstarttime 0)
    (return false 0.0)
    (else
      (return true (Math|Float|Clamp(Float) (/ (- (Fortnite|Sync|Time|GetServerWorldTimeSeconds) _introsequenceserverstarttime) (ScalableFloat|GetValueatLevel (Variables|Default|GetHF_IntroSequenceDuration))))))))

(fn GetIsVehicleHiddenLocal
  (return (Variables|Default|GetIsVehicleHiddenLocal)))

(fn SpawnLoot
  (bind _attachedvehicle (Variables|Default|GetAttachedVehicle)
    (:then
      (Vehicle|Mod|ApplyVehicleMod _attachedvehicle (Variables|Default|GetLootCrateMod) (Fortnite|GetFortPlayerControllerfromActor (Vehicle|GetDriver _attachedvehicle))))
    (:else)))

(fn UpdatePassengerHiddenStatus
  (params (PassengerPawn :type Fort%20Player%20Pawn%20Object%20Reference) (NewIsHidden :type Boolean))
  (Utilities|IsValid PassengerPawn
    (:Is%20Valid
      (Utilities|FlowControl|Sequence
        (:then_0
          (Networking|SwitchHasAuthority
            (:Authority
              (Variables|GetNewIsHidden
                (:then
                  (bind _passengerpawn (Variables|GetPassengerPawn)
                    (:then
                      (Utilities|IsValid (Ability|GetAbilitySystemComponent _passengerpawn)
                        (:Is%20Valid
                          (GameplayEffects|ApplyGameplayEffectToTarget :self (Ability|GetAbilitySystemComponent _passengerpawn) :GameplayEffectClass "/GreaseClipDunmurry/Gameplay/GameplayEffects/GE_GreaseClipDunmurry_Hidden.GE_GreaseClipDunmurry_Hidden_C" :Target (Ability|GetAbilitySystemComponent _passengerpawn) :Level 0.0))))
                    (:else)))
                (:else
                  (bind _passengerpawn_2 (Variables|GetPassengerPawn)
                    (:then
                      (Utilities|IsValid (Ability|GetAbilitySystemComponent _passengerpawn_2)
                        (:Is%20Valid
                          (GameplayEffects|RemoveActiveEffectswithTags (Ability|GetAbilitySystemComponent _passengerpawn_2) "(GameplayTags=((TagName=\"Pawn.HIdePlayerPawn.GreaseClipDunmurry\")))"))))
                    (:else)))))))
        (:then_1
          (bind _passengerpawn_3 (Variables|GetPassengerPawn)
            (:then
              (Collision|SetActorEnableCollision _passengerpawn_3 (not (Variables|GetNewIsHidden))))
            (:else)))))))

(fn UpdateVehicleHiddenStatus
  (bind _returnvalue (CallFunction|IsVehicleHidden))
  (bind _ownervehicle (LocalVariables|GetOwnerVehicle))
  (bind _ownervehicle_2 (LocalVariables|GetOwnerVehicle))
  (if (!= (Variables|Default|GetIsVehicleHiddenLocal) _returnvalue)
    (bind _asfort_athena_vehicle (Utilities|Casting|CastToFortAthenaVehicle (Actor|GetOwner)))
    (LocalVariables|SetOwnerVehicle _asfort_athena_vehicle)
    (bind _output_get (Variables|Default|SetIsVehicleHiddenLocal _returnvalue))
    (if _output_get
      (Rendering|SetActorHiddenInGame _ownervehicle_2 true)
      (Collision|SetActorEnableCollision _ownervehicle_2 false)
      (for _array_element (Vehicle|GetAllPassengers _ownervehicle_2)
        (CallFunction|UpdatePassengerHiddenStatus _array_element true))
      (CallFunction|HaltVehicle)
      (bind _returnvalue_2 (Utilities|Time|SetTimerbyEvent (EventDispatchers|CreateEvent :!event HaltVehicle) 0.1 true))
      (Variables|Default|SetHaltVehicleTimerHandle _returnvalue_2)
      (Networking|SwitchHasAuthority
        (:Authority
          (bind _returnvalue_3 (GameplayEffects|ApplyGameplayEffectToTarget :self (Ability|GetAbilitySystemComponent (LocalVariables|GetOwnerVehicle)) :GameplayEffectClass "/GreaseClipDunmurry/Gameplay/GameplayEffects/GE_GreaseClipDunmurry_VehicleHidden.GE_GreaseClipDunmurry_VehicleHidden_C" :Target (Ability|GetAbilitySystemComponent (LocalVariables|GetOwnerVehicle)) :Level 0.0))
          (Variables|Default|SetVehicleHiddenEffectHandle _returnvalue_3)))
      (else
        (Rendering|SetActorHiddenInGame _ownervehicle false)
        (Collision|SetActorEnableCollision _ownervehicle true)
        (for _array_element_2 (Vehicle|GetAllPassengers _ownervehicle)
          (CallFunction|UpdatePassengerHiddenStatus _array_element_2))
        (Utilities|Time|ClearandInvalidateTimerbyHandle (Variables|Default|GetHaltVehicleTimerHandle))
        (CallFunction|RestoreVehicleVelocity)
        (Networking|SwitchHasAuthority
          (:Authority
            (GameplayEffects|RemoveActiveGameplayEffect (Ability|GetAbilitySystemComponent (LocalVariables|GetOwnerVehicle)) (Variables|Default|GetVehicleHiddenEffectHandle) -1)
            (Variables|Default|SetVehicleHiddenEffectHandle)))))))

(fn ConstructionScript)
